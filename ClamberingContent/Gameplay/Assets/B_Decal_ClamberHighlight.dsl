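; Decal Blueprint for a clamber highlight, showing by color whether the clamber is in range.
; SetIsInRange checks that the decal's dynamic material instance is valid, then sets its Color parameter to the in-range or out-of-range color variable.
; Where cosmetic events can execute, the Construction Script gives the decal absolute rotation and zeroes its world rotation, so it ignores parent rotation.
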(fn SetIsInRange
  (params (bInRange? :type Boolean))
  (bind _decal_mid (Variables|Default|GetDecal_MID)
    (:then
      (Rendering|Material|SetVectorParameterValue _decal_mid "Color" (select bInRange? (Variables|Default|GetColor_InRange) (Variables|Default|GetColor_NotInRange)) :!class /Script/Engine.MaterialInstanceDynamic))
    (:else)))

(fn ConstructionScript
  (bind _decal (Variables|Decal|GetDecal))
  (Networking|CanExecuteCosmeticEvents
    (:True
      (Variables|/Script/Engine.SceneComponent|SetbAbsoluteRotation true _decal)
      (Transformation|SetWorldRotation _decal "0, 0, 0" false false))))
