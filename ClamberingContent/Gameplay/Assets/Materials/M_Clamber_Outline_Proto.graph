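# Prototype unlit translucent deferred decal material for clambering, drawing a glowing outline and arrow texture in a boosted, tintable color, orange by default.
# Actor-relative world position feeds sphere-mask lines with soft center fill, plus clamped arrow UVs, built separately for X- and Y-facing surfaces.
# Absolute decal normal weights blend both projections, omitting Z-facing surfaces.
# Looping modulo time and its sine provide animation signals, including a line radius interpolation.
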Material /ClamberingContent/Gameplay/Assets/Materials/M_Clamber_Outline_Proto
graph {
 properties {
  blendMode=BLEND_Translucent
  materialDomain=MD_DeferredDecal
  pixelDepthOffsetMode=PDOM_Legacy
  shadingModel=MSM_Unlit
  usedShadingModels=Unlit
 }
 nodes [
  n1 = Constant() {r=1}
  n2 = ScalarParameter() {defaultValue=30, parameterName=ArrowsDivisorU}
  n3 = ScalarParameter() {defaultValue=15, parameterName=ArrowsDivisorV}
  n4 = ConstantBiasScale(None=Divide(
    A=NamedRerouteUsage()
    B=MaterialFunctionCall(R=@n1, G=@n2, B=@n3) {materialFunction=&AppendMany}.RGB
   ))
  n5 = Constant()
  n6 = ConstantBiasScale(None=NamedRerouteUsage()) {bias=-0.5, scale=2.5}
  n7 = TextureSample(UVs=Add(A=ComponentMask(None=@n4) {b=true, g=true}, B=AppendVector(A=@n5, B=@n6))) {
   samplerSource=SSM_Clamp_WorldGroupSettings
   samplerType=SAMPLERTYPE_LinearColor
   texture=/ClamberingContent/Gameplay/Assets/Textures/T_Clambering_Arrow
  }
  n8 = NamedRerouteDeclaration(None=Subtract(A=ActorPositionWS(), B=WorldPosition().XYZ)) {name=LocalPosition.XYZ, nodeColor=<rgba 0.1719 0 1 1>}
  n9 = ScalarParameter() {defaultValue=50, parameterName=DecalScale}
  n10 = MaterialFunctionCall(Float3=Abs(None=Divide(A=@n8, B=@n9) {constB=100})) {materialFunction=&BreakOutFloat3Components}
  n11 = Max(B=@n10.G)
  n12 = OneMinus(None=Power(Base=@n11, Exp=ScalarParameter() {defaultValue=1.5, parameterName=RampExponent}))
  n13 = Constant()
  n14 = NamedRerouteUsage()
  n15 = ScalarParameter() {defaultValue=0.1, parameterName=LineHardness}
  n16 = Constant()
  n17 = Constant() {r=1}
  n18 = ScalarParameter() {defaultValue=0.5, parameterName=CenterHardness}
  n19 = Add(
   A=Power(Base=SphereMask(A=@n12, B=@n13, Radius=@n14, Hardness=@n15), Exp=ScalarParameter() {defaultValue=8, parameterName=LineExponent})
   B=Multiply(A=SphereMask(A=@n11, B=@n16, Radius=@n17, Hardness=@n18), B=ScalarParameter() {defaultValue=0.1, parameterName=CenterScale})
  )
  n20 = Abs(None=MaterialFunctionCall() {materialFunction=/Game/Weapons/Prototype/SprayGun/Materials/MF_DecalNormal}.WorldNormal)
  n21 = Constant3Vector() {constant=<rgba 1 0 0 0>}
  n22 = Constant3Vector() {constant=<rgba 0 1 0 0>}
  n23 = Max(B=@n10.R)
  n24 = OneMinus(None=Power(Base=@n23, Exp=ScalarParameter() {defaultValue=1.5, parameterName=RampExponent}))
  n25 = Constant()
  n26 = NamedRerouteUsage()
  n27 = ScalarParameter() {defaultValue=0.1, parameterName=LineHardness}
  n28 = Constant()
  n29 = Constant() {r=1}
  n30 = ScalarParameter() {defaultValue=0.5, parameterName=CenterHardness}
  n31 = Add(
   A=Power(Base=SphereMask(A=@n24, B=@n25, Radius=@n26, Hardness=@n27), Exp=ScalarParameter() {defaultValue=8, parameterName=LineExponent})
   B=Multiply(A=SphereMask(A=@n23, B=@n28, Radius=@n29, Hardness=@n30), B=ScalarParameter() {defaultValue=0.1, parameterName=CenterScale})
  )
  n32 = ScalarParameter() {defaultValue=30, parameterName=ArrowsDivisorU}
  n33 = Constant() {r=1}
  n34 = ScalarParameter() {defaultValue=15, parameterName=ArrowsDivisorV}
  n35 = ConstantBiasScale(None=Divide(
    A=NamedRerouteUsage()
    B=MaterialFunctionCall(R=@n32, G=@n33, B=@n34) {materialFunction=&AppendMany}.RGB
   ))
  n36 = Constant()
  n37 = ConstantBiasScale(None=NamedRerouteUsage()) {bias=-0.5, scale=2.5}
  n38 = TextureSample(UVs=Add(A=ComponentMask(None=@n35) {b=true, r=true}, B=AppendVector(A=@n36, B=@n37))) {
   samplerSource=SSM_Clamp_WorldGroupSettings
   samplerType=SAMPLERTYPE_LinearColor
   texture=&T_Clambering_Arrow
  }
  Add(A=Multiply(A=Add(A=@n7.R, B=@n19), B=DotProduct(A=@n20, B=@n21)), B=Multiply(A=DotProduct(A=@n20, B=@n22), B=Add(A=@n31, B=@n38.R)))
  Constant()
  Multiply(
   A=VectorParameter() {parameterName=Color, defaultValue=<rgba 1 0.4 0 1>}.RGB
   B=ScalarParameter() {defaultValue=2, parameterName=EmissiveBoost}
  )
  n39 = MaterialFunctionCall!(Time, Divisor=ScalarParameter() {defaultValue=0.55, parameterName=ModuloTimeDivisor}, Mod) {materialFunction=/Game/Effects/Fort_Effects/Materials/_Functions/MF_ModuloTime}
  NamedRerouteDeclaration(None=@n39.Clamped 0-1) {name=ModuloTime, nodeColor=<rgba 0.9453 1 0 1>}
  NamedRerouteDeclaration(None=LinearInterpolate(
    A=ScalarParameter() {defaultValue=0.05, parameterName=LineRadiusA}
    B=ScalarParameter() {defaultValue=0.12, parameterName=LineRadiusB}
    Alpha=NamedRerouteUsage()
   )) {name=LineRadius, nodeColor=<rgba 0.5938 0 1 1>}
  n40 = Sine(None=@n39.Clamped 0-1)
  NamedRerouteDeclaration(None=ConstantBiasScale(None=NamedRerouteDeclaration(None=@n40) {name=SineModuleTime, nodeColor=<rgba 1 0 0.0477 1>})) {name=SineModuleTimeBiased, nodeColor=<rgba 0 1 0.3438 1>}
 ]
}
